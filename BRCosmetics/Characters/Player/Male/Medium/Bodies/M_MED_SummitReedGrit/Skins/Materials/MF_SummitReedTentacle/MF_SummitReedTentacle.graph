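# Material function for the SummitReedGrit skin that draws one layer of animated, wavy tentacles, outputting Mask and Outline.
# LayerInput selects frames from 2x2 flipbook data textures that encode position, angle, scale, masks and randomness, which build recentered, rotated, scaled UVs.
# A time-driven sine wave feeds a SphereMask to shape the tentacle procedurally, with A/B inputs blending speed, frequency, amplitude, thickness and length.
# Subtracting a smaller-radius SphereMask produces the outline.
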MaterialFunction /BRCosmetics/Characters/Player/Male/Medium/Bodies/M_MED_SummitReedGrit/Skins/Materials/MF_SummitReedTentacle/MF_SummitReedTentacle
graph {expressions {
  MaterialExpressionAdd_14 {
   class=MaterialExpressionAdd
   inputs [
    A {from=MaterialExpressionMultiply_43}
    B {from=MaterialExpressionMultiply_44}
   ]
  }
  MaterialExpressionAdd_15 {
   class=MaterialExpressionAdd
   inputs [
    A {from=MaterialExpressionMultiply_45}
    B {from=MaterialExpressionLinearInterpolate_5}
   ]
  }
  MaterialExpressionAppendVector_6 {
   class=MaterialExpressionAppendVector
   inputs [
    A {from=MaterialExpressionNamedRerouteUsage_24, output=""}
    B {from=MaterialExpressionOneMinus_6, output=""}
   ]
  }
  MaterialExpressionAppendVector_7 {
   class=MaterialExpressionAppendVector
   inputs [
    A {from=MaterialExpressionConvert_6, output=""}
    B {from=MaterialExpressionConvert_6, output=""}
   ]
  }
  MaterialExpressionComponentMask_2 {class=MaterialExpressionComponentMask, inputs [None {from=MaterialExpressionMaterialFunctionCall_5, output=Result}], properties {g=true}}
  MaterialExpressionConstant_0 {class=MaterialExpressionConstant, properties {r=2}}
  MaterialExpressionConstant_1 {class=MaterialExpressionConstant, properties {r=2}}
  MaterialExpressionConstant_10 {class=MaterialExpressionConstant, properties {r=0.6}}
  MaterialExpressionConstant_11 {class=MaterialExpressionConstant, properties {r=0.3}}
  MaterialExpressionConstant_12 {class=MaterialExpressionConstant, properties {r=0.9158}}
  MaterialExpressionConstant_13 {class=MaterialExpressionConstant, properties {r=1}}
  MaterialExpressionConstant_14 {class=MaterialExpressionConstant, properties {r=5.849}}
  MaterialExpressionConstant_15 {class=MaterialExpressionConstant, properties {r=4}}
  MaterialExpressionConstant_16 {class=MaterialExpressionConstant, properties {r=0.02}}
  MaterialExpressionConstant_17 {class=MaterialExpressionConstant, properties {r=-0.122}}
  MaterialExpressionConstant_18 {class=MaterialExpressionConstant, properties {r=-0.4}}
  MaterialExpressionConstant_2 {class=MaterialExpressionConstant, properties {r=2}}
  MaterialExpressionConstant_3 {class=MaterialExpressionConstant, properties {r=2}}
  MaterialExpressionConstant_4 {class=MaterialExpressionConstant, properties {r=4}}
  MaterialExpressionConstant_40 {class=MaterialExpressionConstant, properties {r=0.5}}
  MaterialExpressionConstant_5 {class=MaterialExpressionConstant, properties {r=2}}
  MaterialExpressionConstant_55 {class=MaterialExpressionConstant}
  MaterialExpressionConstant_6 {class=MaterialExpressionConstant, properties {r=2.5}}
  MaterialExpressionConstant_63 {class=MaterialExpressionConstant, properties {r=-1}}
  MaterialExpressionConstant_64 {class=MaterialExpressionConstant, properties {r=1}}
  MaterialExpressionConstant_69 {class=MaterialExpressionConstant, properties {r=0.5}}
  MaterialExpressionConstant_7 {class=MaterialExpressionConstant, properties {r=1.2}}
  MaterialExpressionConstant_8 {class=MaterialExpressionConstant, properties {r=-0.3}}
  MaterialExpressionConstant_9 {class=MaterialExpressionConstant, properties {r=-0.4}}
  MaterialExpressionConvert_4 {
   class=MaterialExpressionConvert
   inputs [None {from=MaterialExpressionMaterialFunctionCall_0, output=Result}]
   properties {
    nodeName=Break Float3
    convertInputs [{
      type=Vector3
      expressionInput {
       expression=&self:MaterialExpressionMaterialFunctionCall_0
       inputName=None
       mask=0
       maskA=0
       maskB=0
       maskG=0
       maskR=0
       outputIndex=0
      }
     }]
    convertOutputs [{type=Scalar}, {type=Scalar}, {type=Scalar}]
   }
  }
  MaterialExpressionConvert_5 {
   class=MaterialExpressionConvert
   inputs [None {from=MaterialExpressionMaterialFunctionCall_3, output=UVs}]
   properties {
    nodeName=Break Float2
    convertInputs [{
      type=Vector2
      expressionInput {
       expression=&self:MaterialExpressionMaterialFunctionCall_3
       inputName=None
       mask=0
       maskA=0
       maskB=0
       maskG=0
       maskR=0
       outputIndex=0
      }
     }]
    convertOutputs [{type=Scalar}, {type=Scalar}]
   }
  }
  MaterialExpressionConvert_6 {
   class=MaterialExpressionConvert
   inputs [None {from=MaterialExpressionMaterialFunctionCall_2, output=Result}]
   properties {
    nodeName=Break Float3
    convertInputs [{
      type=Vector3
      expressionInput {
       expression=&self:MaterialExpressionMaterialFunctionCall_2
       inputName=None
       mask=0
       maskA=0
       maskB=0
       maskG=0
       maskR=0
       outputIndex=0
      }
     }]
    convertOutputs [{type=Scalar}, {type=Scalar}, {type=Scalar}]
   }
  }
  MaterialExpressionConvert_7 {
   class=MaterialExpressionConvert
   inputs [None {from=MaterialExpressionMaterialFunctionCall_4, output=Result}]
   properties {
    nodeName=Break Float3
    convertInputs [{
      type=Vector3
      expressionInput {
       expression=&self:MaterialExpressionMaterialFunctionCall_4
       inputName=None
       mask=0
       maskA=0
       maskB=0
       maskG=0
       maskR=0
       outputIndex=0
      }
     }]
    convertOutputs [{type=Scalar}, {type=Scalar}, {type=Scalar}]
   }
  }
  MaterialExpressionDivide_0 {
   class=MaterialExpressionDivide
   inputs [
    A {from=MaterialExpressionSubtract_1}
    B {from=""}
   ]
   properties {constB=4}
  }
  MaterialExpressionDivide_1 {
   class=MaterialExpressionDivide
   inputs [
    A {from=MaterialExpressionSubtract_2}
    B {from=""}
   ]
   properties {constB=4}
  }
  MaterialExpressionDivide_2 {
   class=MaterialExpressionDivide
   inputs [
    A {from=MaterialExpressionSubtract_3}
    B {from=""}
   ]
   properties {constB=4}
  }
  MaterialExpressionDivide_3 {
   class=MaterialExpressionDivide
   inputs [
    A {from=MaterialExpressionSubtract_5}
    B {from=""}
   ]
   properties {constB=4}
  }
  MaterialExpressionFunctionInput_0 {
   class=MaterialExpressionFunctionInput
   inputs [Preview {from=MaterialExpressionConstant_9, output=""}]
   properties {bUsePreviewValueAsDefault=true, inputName=SpeedA, inputType=FunctionInput_Scalar}
  }
  MaterialExpressionFunctionInput_1 {
   class=MaterialExpressionFunctionInput
   inputs [Preview {from=MaterialExpressionConstant_8, output=""}]
   properties {bUsePreviewValueAsDefault=true, inputName=SpeedB}
  }
  MaterialExpressionFunctionInput_10 {
   class=MaterialExpressionFunctionInput
   inputs [Preview {from=MaterialExpressionConstant_17, output=""}]
   properties {bUsePreviewValueAsDefault=true, inputName=LengthB, inputType=FunctionInput_Scalar}
  }
  MaterialExpressionFunctionInput_11 {
   class=MaterialExpressionFunctionInput
   inputs [Preview {from=MaterialExpressionConstant_16, output=""}]
   properties {bUsePreviewValueAsDefault=true, inputName=LayerOutline, inputType=FunctionInput_Scalar}
  }
  MaterialExpressionFunctionInput_12 {
   class=MaterialExpressionFunctionInput
   inputs [Preview {from=MaterialExpressionConstant_10, output=""}]
   properties {bUsePreviewValueAsDefault=true, inputName=AmplitudeA, inputType=FunctionInput_Scalar}
  }
  MaterialExpressionFunctionInput_13 {
   class=MaterialExpressionFunctionInput
   inputs [Preview {from=MaterialExpressionConstant_11, output=""}]
   properties {bUsePreviewValueAsDefault=true, inputName=AmplitudeB, inputType=FunctionInput_Scalar}
  }
  MaterialExpressionFunctionInput_14 {
   class=MaterialExpressionFunctionInput
   inputs [Preview {from=MaterialExpressionConstant_7, output=""}]
   properties {bUsePreviewValueAsDefault=true, inputName=SpeedBoost}
  }
  MaterialExpressionFunctionInput_2 {
   class=MaterialExpressionFunctionInput
   inputs [Preview {from=MaterialExpressionConstant_6, output=""}]
   properties {bUsePreviewValueAsDefault=true, inputName=FrequencyA, inputType=FunctionInput_Scalar}
  }
  MaterialExpressionFunctionInput_3 {
   class=MaterialExpressionFunctionInput
   inputs [Preview {from=MaterialExpressionConstant_0, output=""}]
   properties {bUsePreviewValueAsDefault=true, inputName=FrequencyB, inputType=FunctionInput_Scalar}
  }
  MaterialExpressionFunctionInput_4 {
   class=MaterialExpressionFunctionInput
   inputs [Preview {from=MaterialExpressionConstant_14, output=""}]
   properties {bUsePreviewValueAsDefault=true, inputName=Thickness and fineness A, inputType=FunctionInput_Scalar}
  }
  MaterialExpressionFunctionInput_5 {
   class=MaterialExpressionFunctionInput
   inputs [Preview {from=MaterialExpressionScalarParameter_0, output=""}]
   properties {bUsePreviewValueAsDefault=true, inputName=LayerInput, inputType=FunctionInput_Scalar}
  }
  MaterialExpressionFunctionInput_6 {
   class=MaterialExpressionFunctionInput
   inputs [Preview {from=MaterialExpressionConstant_15, output=""}]
   properties {bUsePreviewValueAsDefault=true, inputName=Thickness and fineness B, inputType=FunctionInput_Scalar}
  }
  MaterialExpressionFunctionInput_7 {
   class=MaterialExpressionFunctionInput
   inputs [Preview {from=MaterialExpressionConstant_12, output=""}]
   properties {bUsePreviewValueAsDefault=true, inputName=Thickness and fineness C, inputType=FunctionInput_Scalar}
  }
  MaterialExpressionFunctionInput_8 {
   class=MaterialExpressionFunctionInput
   inputs [Preview {from=MaterialExpressionConstant_18, output=""}]
   properties {bUsePreviewValueAsDefault=true, inputName=LengthA, inputType=FunctionInput_Scalar}
  }
  MaterialExpressionFunctionInput_9 {
   class=MaterialExpressionFunctionInput
   inputs [Preview {from=MaterialExpressionConstant_13, output=""}]
   properties {bUsePreviewValueAsDefault=true, inputName=Amplitude boost}
  }
  MaterialExpressionFunctionOutput_0 {
   class=MaterialExpressionFunctionOutput
   inputs [None {from=MaterialExpressionNamedRerouteDeclaration_20, output=""}]
   properties {outputName=Mask}
  }
  MaterialExpressionFunctionOutput_1 {class=MaterialExpressionFunctionOutput, inputs [None {from=MaterialExpressionMultiply_0, output=""}], properties {outputName=Outline}}
  MaterialExpressionIf_2 {
   class=MaterialExpressionIf
   inputs [
    A {from=MaterialExpressionNamedRerouteUsage_26, output=""}
    B {from="", output=""}
    "A > B" {from=MaterialExpressionConstant_63, output=""}
    "A == B" {from=MaterialExpressionConstant_63, output=""}
    "A < B" {from=MaterialExpressionConstant_64, output=""}
   ]
   properties {constB=0.02}
  }
  MaterialExpressionLinearInterpolate_0 {
   class=MaterialExpressionLinearInterpolate
   inputs [
    A {from=""}
    B {from=""}
    Alpha {from=MaterialExpressionNamedRerouteUsage_3}
   ]
   properties {constA=0.8}
  }
  MaterialExpressionLinearInterpolate_10 {
   class=MaterialExpressionLinearInterpolate
   inputs [
    A {from=MaterialExpressionFunctionInput_12}
    B {from=MaterialExpressionFunctionInput_13}
    Alpha {from=MaterialExpressionNamedRerouteUsage_23}
   ]
   properties {constA=0.6, constB=0.8}
  }
  MaterialExpressionLinearInterpolate_5 {
   class=MaterialExpressionLinearInterpolate
   inputs [
    A {from=MaterialExpressionFunctionInput_8}
    B {from=MaterialExpressionFunctionInput_10}
    Alpha {from=MaterialExpressionNamedRerouteUsage_22}
   ]
  }
  MaterialExpressionLinearInterpolate_7 {
   class=MaterialExpressionLinearInterpolate
   inputs [
    A {from=MaterialExpressionFunctionInput_2}
    B {from=MaterialExpressionFunctionInput_3}
    Alpha {from=MaterialExpressionNamedRerouteUsage_30}
   ]
   properties {constA=2.5, constB=2}
  }
  MaterialExpressionLinearInterpolate_8 {
   class=MaterialExpressionLinearInterpolate
   inputs [
    A {from=MaterialExpressionFunctionInput_0}
    B {from=MaterialExpressionFunctionInput_1}
    Alpha {from=MaterialExpressionNamedRerouteUsage_32}
   ]
   properties {constA=-0.4, constB=-0.2}
  }
  MaterialExpressionLinearInterpolate_9 {
   class=MaterialExpressionLinearInterpolate
   inputs [
    A {from=MaterialExpressionFunctionInput_4}
    B {from=MaterialExpressionFunctionInput_6}
    Alpha {from=MaterialExpressionNamedRerouteUsage_21}
   ]
   properties {constB=2}
  }
  MaterialExpressionMaterialFunctionCall_0 {
   class=MaterialExpressionMaterialFunctionCall
   inputs [
    "Animation  Phase (0-1)" {from=MaterialExpressionSaturate_0}
    Number of Rows {from=MaterialExpressionConstant_1}
    Number of Columns {from=MaterialExpressionConstant_1}
    Texture {from=MaterialExpressionTextureObjectParameter_1}
    UVs {from=MaterialExpressionTextureCoordinate_0}
    MipBias/Level {from=""}
    "Use Mip Bias (T) Level (F)" {from=""}
    "Clamp Anim (see tooltip)" {from=""}
   ]
   properties {materialFunction=&FlipBook}
  }
  MaterialExpressionMaterialFunctionCall_1 {
   class=MaterialExpressionMaterialFunctionCall
   inputs [
    UVs {from=MaterialExpressionSubtract_19}
    Rotation Center {from=MaterialExpressionConstant_40}
    "Rotation Angle (0-1)" {from=MaterialExpressionMultiply_27}
   ]
   properties {materialFunction=&CustomRotator}
  }
  MaterialExpressionMaterialFunctionCall_2 {
   class=MaterialExpressionMaterialFunctionCall
   inputs [
    "Animation  Phase (0-1)" {from=MaterialExpressionSaturate_1}
    Number of Rows {from=MaterialExpressionConstant_2}
    Number of Columns {from=MaterialExpressionConstant_2}
    Texture {from=MaterialExpressionTextureObjectParameter_0}
    UVs {from=MaterialExpressionTextureCoordinate_1}
    MipBias/Level {from=""}
    "Use Mip Bias (T) Level (F)" {from=""}
    "Clamp Anim (see tooltip)" {from=""}
   ]
   properties {materialFunction=&FlipBook}
  }
  MaterialExpressionMaterialFunctionCall_3 {
   class=MaterialExpressionMaterialFunctionCall
   inputs [
    UVs {from=MaterialExpressionMaterialFunctionCall_1, output=Rotated Values}
    Texture Scale {from=MaterialExpressionNamedRerouteUsage_28}
   ]
   properties {materialFunction=&ScaleUVsByCenter}
  }
  MaterialExpressionMaterialFunctionCall_4 {
   class=MaterialExpressionMaterialFunctionCall
   inputs [
    "Animation  Phase (0-1)" {from=MaterialExpressionSaturate_2}
    Number of Rows {from=MaterialExpressionConstant_3}
    Number of Columns {from=MaterialExpressionConstant_3}
    Texture {from=MaterialExpressionTextureObjectParameter_2}
    UVs {from=MaterialExpressionTextureCoordinate_2}
    MipBias/Level {from=""}
    "Use Mip Bias (T) Level (F)" {from=""}
    "Clamp Anim (see tooltip)" {from=""}
   ]
   properties {materialFunction=&FlipBook}
  }
  MaterialExpressionMaterialFunctionCall_5 {
   class=MaterialExpressionMaterialFunctionCall
   inputs [
    "Animation  Phase (0-1)" {from=MaterialExpressionSaturate_3}
    Number of Rows {from=MaterialExpressionConstant_5}
    Number of Columns {from=MaterialExpressionConstant_5}
    Texture {from=MaterialExpressionTextureObjectParameter_3}
    UVs {from=MaterialExpressionTextureCoordinate_3}
    MipBias/Level {from=""}
    "Use Mip Bias (T) Level (F)" {from=""}
    "Clamp Anim (see tooltip)" {from=""}
   ]
   properties {materialFunction=&FlipBook}
  }
  MaterialExpressionMultiply_0 {
   class=MaterialExpressionMultiply
   inputs [
    A {from=MaterialExpressionSubtract_0}
    B {from=MaterialExpressionNamedRerouteUsage_1}
   ]
  }
  MaterialExpressionMultiply_1 {
   class=MaterialExpressionMultiply
   inputs [
    A {from=MaterialExpressionSphereMask_6}
    B {from=MaterialExpressionLinearInterpolate_0}
   ]
  }
  MaterialExpressionMultiply_27 {
   class=MaterialExpressionMultiply
   inputs [
    A {from=MaterialExpressionNamedRerouteUsage_27}
    B {from=MaterialExpressionIf_2}
   ]
  }
  MaterialExpressionMultiply_3 {
   class=MaterialExpressionMultiply
   inputs [
    A {from=MaterialExpressionMultiply_54}
    B {from=MaterialExpressionOneMinus_1}
   ]
  }
  MaterialExpressionMultiply_4 {
   class=MaterialExpressionMultiply
   inputs [
    A {from=MaterialExpressionLinearInterpolate_10}
    B {from=MaterialExpressionFunctionInput_9}
   ]
  }
  MaterialExpressionMultiply_40 {
   class=MaterialExpressionMultiply
   inputs [
    A {from=MaterialExpressionTime_4}
    B {from=MaterialExpressionMultiply_5}
   ]
   properties {constB=0.4}
  }
  MaterialExpressionMultiply_41 {
   class=MaterialExpressionMultiply
   inputs [
    A {from=MaterialExpressionLinearInterpolate_7}
    B {from=MaterialExpressionNamedRerouteUsage_37}
   ]
   properties {constA=1.5}
  }
  MaterialExpressionMultiply_42 {
   class=MaterialExpressionMultiply
   inputs [
    A {from=MaterialExpressionSine_4}
    B {from=MaterialExpressionNamedRerouteUsage_35}
   ]
  }
  MaterialExpressionMultiply_43 {
   class=MaterialExpressionMultiply
   inputs [
    A {from=MaterialExpressionMultiply_42}
    B {from=MaterialExpressionMultiply_4}
   ]
  }
  MaterialExpressionMultiply_44 {
   class=MaterialExpressionMultiply
   inputs [
    A {from=MaterialExpressionSubtract_22}
    B {from=MaterialExpressionLinearInterpolate_9}
   ]
   properties {constB=8}
  }
  MaterialExpressionMultiply_45 {
   class=MaterialExpressionMultiply
   inputs [
    A {from=MaterialExpressionNamedRerouteUsage_36}
    B {from=MaterialExpressionFunctionInput_7}
   ]
  }
  MaterialExpressionMultiply_5 {
   class=MaterialExpressionMultiply
   inputs [
    A {from=MaterialExpressionLinearInterpolate_8}
    B {from=MaterialExpressionFunctionInput_14}
   ]
   properties {constB=2}
  }
  MaterialExpressionMultiply_54 {
   class=MaterialExpressionMultiply
   inputs [
    A {from=MaterialExpressionSphereMask_6}
    B {from=MaterialExpressionNamedRerouteUsage_31}
   ]
  }
  MaterialExpressionNamedRerouteDeclaration_10 {
   class=MaterialExpressionNamedRerouteDeclaration
   inputs [None {from=MaterialExpressionMaterialFunctionCall_0, output=Alpha}]
   properties {name=Pos_G, nodeColor=<rgba 0.9219 0 1 1>}
  }
  MaterialExpressionNamedRerouteDeclaration_11 {
   class=MaterialExpressionNamedRerouteDeclaration
   inputs [None {from=MaterialExpressionConvert_4}]
   properties {name=Angle_R, nodeColor=<rgba 0 1 0.6016 1>}
  }
  MaterialExpressionNamedRerouteDeclaration_12 {
   class=MaterialExpressionNamedRerouteDeclaration
   inputs [None {from=MaterialExpressionConvert_4}]
   properties {name=Angle_G, nodeColor=<rgba 1 0.2812 0 1>}
  }
  MaterialExpressionNamedRerouteDeclaration_13 {
   class=MaterialExpressionNamedRerouteDeclaration
   inputs [None {from=MaterialExpressionAppendVector_7}]
   properties {name=Scale, nodeColor=<rgba 0.3594 0 1 1>}
  }
  MaterialExpressionNamedRerouteDeclaration_16 {
   class=MaterialExpressionNamedRerouteDeclaration
   inputs [None {from=MaterialExpressionSwitch_0}]
   properties {name=Mask, nodeColor=<rgba 0 0.3672 1 1>}
  }
  MaterialExpressionNamedRerouteDeclaration_17 {
   class=MaterialExpressionNamedRerouteDeclaration
   inputs [None {from=MaterialExpressionConvert_5}]
   properties {name=U, nodeColor=<rgba 1 0 0.375 1>}
  }
  MaterialExpressionNamedRerouteDeclaration_18 {
   class=MaterialExpressionNamedRerouteDeclaration
   inputs [None {from=MaterialExpressionConvert_5}]
   properties {name=V, nodeColor=<rgba 0 0.6953 1 1>}
  }
  MaterialExpressionNamedRerouteDeclaration_19 {
   class=MaterialExpressionNamedRerouteDeclaration
   inputs [None {from=MaterialExpressionMultiply_41}]
   properties {name=Frequency, nodeColor=<rgba 1 0 0.3047 1>}
  }
  MaterialExpressionNamedRerouteDeclaration_20 {
   class=MaterialExpressionNamedRerouteDeclaration
   inputs [None {from=MaterialExpressionMultiply_3}]
   properties {name=Layer1, nodeColor=<rgba 0 0.625 1 1>}
  }
  MaterialExpressionNamedRerouteDeclaration_6 {
   class=MaterialExpressionNamedRerouteDeclaration
   inputs [None {from=MaterialExpressionFunctionInput_5}]
   properties {name=LayerInput, nodeColor=<rgba 0.6875 1 0 1>}
  }
  MaterialExpressionNamedRerouteDeclaration_7 {
   class=MaterialExpressionNamedRerouteDeclaration
   inputs [None {from=MaterialExpressionMaterialFunctionCall_4, output=Alpha}]
   properties {name=FXMask, nodeColor=<rgba 0.8281 0 1 1>}
  }
  MaterialExpressionNamedRerouteDeclaration_8 {
   class=MaterialExpressionNamedRerouteDeclaration
   inputs [None {from=MaterialExpressionComponentMask_2}]
   properties {name=random, nodeColor=<rgba 0.3359 0 1 1>}
  }
  MaterialExpressionNamedRerouteDeclaration_9 {
   class=MaterialExpressionNamedRerouteDeclaration
   inputs [None {from=MaterialExpressionConvert_4}]
   properties {name=POS_R, nodeColor=<rgba 0.7578 1 0 1>}
  }
  MaterialExpressionNamedRerouteUsage_1 {class=MaterialExpressionNamedRerouteUsage}
  MaterialExpressionNamedRerouteUsage_21 {class=MaterialExpressionNamedRerouteUsage}
  MaterialExpressionNamedRerouteUsage_22 {class=MaterialExpressionNamedRerouteUsage}
  MaterialExpressionNamedRerouteUsage_23 {class=MaterialExpressionNamedRerouteUsage}
  MaterialExpressionNamedRerouteUsage_24 {class=MaterialExpressionNamedRerouteUsage}
  MaterialExpressionNamedRerouteUsage_25 {class=MaterialExpressionNamedRerouteUsage}
  MaterialExpressionNamedRerouteUsage_26 {class=MaterialExpressionNamedRerouteUsage}
  MaterialExpressionNamedRerouteUsage_27 {class=MaterialExpressionNamedRerouteUsage}
  MaterialExpressionNamedRerouteUsage_28 {class=MaterialExpressionNamedRerouteUsage}
  MaterialExpressionNamedRerouteUsage_3 {class=MaterialExpressionNamedRerouteUsage}
  MaterialExpressionNamedRerouteUsage_30 {class=MaterialExpressionNamedRerouteUsage}
  MaterialExpressionNamedRerouteUsage_31 {class=MaterialExpressionNamedRerouteUsage}
  MaterialExpressionNamedRerouteUsage_32 {class=MaterialExpressionNamedRerouteUsage}
  MaterialExpressionNamedRerouteUsage_34 {class=MaterialExpressionNamedRerouteUsage}
  MaterialExpressionNamedRerouteUsage_35 {class=MaterialExpressionNamedRerouteUsage}
  MaterialExpressionNamedRerouteUsage_36 {class=MaterialExpressionNamedRerouteUsage}
  MaterialExpressionNamedRerouteUsage_37 {class=MaterialExpressionNamedRerouteUsage}
  MaterialExpressionNamedRerouteUsage_38 {class=MaterialExpressionNamedRerouteUsage}
  MaterialExpressionNamedRerouteUsage_4 {class=MaterialExpressionNamedRerouteUsage}
  MaterialExpressionNamedRerouteUsage_6 {class=MaterialExpressionNamedRerouteUsage}
  MaterialExpressionNamedRerouteUsage_7 {class=MaterialExpressionNamedRerouteUsage}
  MaterialExpressionOneMinus_1 {class=MaterialExpressionOneMinus, inputs [None {from=MaterialExpressionNamedRerouteUsage_6}]}
  MaterialExpressionOneMinus_6 {class=MaterialExpressionOneMinus, inputs [None {from=MaterialExpressionNamedRerouteUsage_25}]}
  MaterialExpressionSaturate_0 {class=MaterialExpressionSaturate, inputs [None {from=MaterialExpressionDivide_0}]}
  MaterialExpressionSaturate_1 {class=MaterialExpressionSaturate, inputs [None {from=MaterialExpressionDivide_1}]}
  MaterialExpressionSaturate_2 {class=MaterialExpressionSaturate, inputs [None {from=MaterialExpressionDivide_2}]}
  MaterialExpressionSaturate_3 {class=MaterialExpressionSaturate, inputs [None {from=MaterialExpressionDivide_3}]}
  MaterialExpressionScalarParameter_0 {class=MaterialExpressionScalarParameter, properties {defaultValue=3, group=Tentacle, parameterName=UVOffset}}
  MaterialExpressionSine_4 {class=MaterialExpressionSine, inputs [None {from=MaterialExpressionSubtract_21}]}
  MaterialExpressionSphereMask_0 {
   class=MaterialExpressionSphereMask
   inputs [
    A {from=MaterialExpressionAdd_14, output=""}
    B {from=MaterialExpressionConstant_55, output=""}
    Radius {from=MaterialExpressionSubtract_7, output=""}
    Hardness {from="", output=""}
   ]
   properties {}
  }
  MaterialExpressionSphereMask_6 {
   class=MaterialExpressionSphereMask
   inputs [
    A {from=MaterialExpressionAdd_14, output=""}
    B {from=MaterialExpressionConstant_55, output=""}
    Radius {from=MaterialExpressionAdd_15, output=""}
    Hardness {from="", output=""}
   ]
   properties {}
  }
  MaterialExpressionSubtract_0 {
   class=MaterialExpressionSubtract
   inputs [
    A {from=MaterialExpressionSphereMask_6}
    B {from=MaterialExpressionSphereMask_0}
   ]
  }
  MaterialExpressionSubtract_1 {
   class=MaterialExpressionSubtract
   inputs [
    A {from=MaterialExpressionNamedRerouteDeclaration_6}
    B {from=""}
   ]
  }
  MaterialExpressionSubtract_19 {
   class=MaterialExpressionSubtract
   inputs [
    A {from=MaterialExpressionTextureCoordinate_12}
    B {from=MaterialExpressionSubtract_20}
   ]
  }
  MaterialExpressionSubtract_2 {
   class=MaterialExpressionSubtract
   inputs [
    A {from=MaterialExpressionNamedRerouteDeclaration_6}
    B {from=""}
   ]
  }
  MaterialExpressionSubtract_20 {
   class=MaterialExpressionSubtract
   inputs [
    A {from=MaterialExpressionAppendVector_6}
    B {from=""}
   ]
   properties {constB=0.5}
  }
  MaterialExpressionSubtract_21 {
   class=MaterialExpressionSubtract
   inputs [
    A {from=MaterialExpressionMultiply_40}
    B {from=MaterialExpressionNamedRerouteUsage_38}
   ]
  }
  MaterialExpressionSubtract_22 {
   class=MaterialExpressionSubtract
   inputs [
    A {from=MaterialExpressionNamedRerouteUsage_34}
    B {from=MaterialExpressionConstant_69}
   ]
  }
  MaterialExpressionSubtract_3 {
   class=MaterialExpressionSubtract
   inputs [
    A {from=MaterialExpressionConstant_4}
    B {from=""}
   ]
  }
  MaterialExpressionSubtract_4 {
   class=MaterialExpressionSubtract
   inputs [
    A {from=MaterialExpressionNamedRerouteUsage_4}
    B {from=""}
   ]
  }
  MaterialExpressionSubtract_5 {
   class=MaterialExpressionSubtract
   inputs [
    A {from=MaterialExpressionNamedRerouteUsage_7}
    B {from=""}
   ]
  }
  MaterialExpressionSubtract_7 {
   class=MaterialExpressionSubtract
   inputs [
    A {from=MaterialExpressionAdd_15}
    B {from=MaterialExpressionFunctionInput_11}
   ]
   properties {constB=0.2}
  }
  MaterialExpressionSwitch_0 {
   class=MaterialExpressionSwitch
   inputs [
    SwitchValue {from=MaterialExpressionSubtract_4, output=""}
    Default {from="", output=""}
    None {from=MaterialExpressionConvert_7, output=""}
    None {from=MaterialExpressionConvert_7, output=""}
    None {from=MaterialExpressionConvert_7, output=""}
   ]
   properties {inputs [
     {
      inputName=None
      input {
       expression=&self:MaterialExpressionConvert_7
       inputName=None
       mask=0
       maskA=0
       maskB=0
       maskG=0
       maskR=0
       outputIndex=0
      }
     }
     {
      inputName=None
      input {
       expression=&self:MaterialExpressionConvert_7
       inputName=None
       mask=0
       maskA=0
       maskB=0
       maskG=0
       maskR=0
       outputIndex=1
      }
     }
     {
      inputName=None
      input {
       expression=&self:MaterialExpressionConvert_7
       inputName=None
       mask=0
       maskA=0
       maskB=0
       maskG=0
       maskR=0
       outputIndex=2
      }
     }
    ]}
  }
  MaterialExpressionTextureCoordinate_0 {class=MaterialExpressionTextureCoordinate}
  MaterialExpressionTextureCoordinate_1 {class=MaterialExpressionTextureCoordinate}
  MaterialExpressionTextureCoordinate_12 {class=MaterialExpressionTextureCoordinate}
  MaterialExpressionTextureCoordinate_2 {class=MaterialExpressionTextureCoordinate}
  MaterialExpressionTextureCoordinate_3 {class=MaterialExpressionTextureCoordinate}
  MaterialExpressionTextureObjectParameter_0 {
   class=MaterialExpressionTextureObjectParameter
   inputs [
    UVs {from="", output=""}
    Apply View MipBias {from="", output=""}
   ]
   properties {
    group=Tentacle
    parameterName=UVInfo_sale
    samplerType=SAMPLERTYPE_Masks
    texture=/BRCosmetics/Characters/Player/Male/Medium/Bodies/M_MED_SummitReedGrit/Skins/Textures/Animation_Texture/T_SummitReedTentacle_uv_2
   }
  }
  MaterialExpressionTextureObjectParameter_1 {
   class=MaterialExpressionTextureObjectParameter
   inputs [
    UVs {from="", output=""}
    Apply View MipBias {from="", output=""}
   ]
   properties {
    group=Tentacle
    parameterName=UVInfo_P_A
    samplerType=SAMPLERTYPE_Masks
    texture=/BRCosmetics/Characters/Player/Male/Medium/Bodies/M_MED_SummitReedGrit/Skins/Textures/Animation_Texture/T_SummitReedTentacle_uv_1
   }
  }
  MaterialExpressionTextureObjectParameter_2 {
   class=MaterialExpressionTextureObjectParameter
   inputs [
    UVs {from="", output=""}
    Apply View MipBias {from="", output=""}
   ]
   properties {
    group=Tentacle
    parameterName=UVInfo_P_A
    samplerType=SAMPLERTYPE_Masks
    texture=&T_SummitReedTentacle_uv_1
   }
  }
  MaterialExpressionTextureObjectParameter_3 {
   class=MaterialExpressionTextureObjectParameter
   inputs [
    UVs {from="", output=""}
    Apply View MipBias {from="", output=""}
   ]
   properties {
    group=Tentacle
    parameterName=UVInfo_RandomColor
    samplerType=SAMPLERTYPE_Masks
    texture=/BRCosmetics/Characters/Player/Male/Medium/Bodies/M_MED_SummitReedGrit/Skins/Textures/Animation_Texture/T_SummitReedTentacle_uv_3
   }
  }
  MaterialExpressionTime_4 {class=MaterialExpressionTime}
 }}
